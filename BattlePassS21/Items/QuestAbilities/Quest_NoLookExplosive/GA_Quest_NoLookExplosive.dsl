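; Quest gameplay ability for the no-look explosive Battle Pass challenge, crediting explosive damage dealt at a particular facing angle.
; On a gameplay event it stores the instigator's player pawn and the target, then checks the instigator tags for explosive damage.
; The owning pawn's yaw must lie within 60 degrees of the inverted instigator-to-target look-at yaw, using a plain range test without angle wrapping.
; A match sends the NoLookExplosive stat to the quest manager, then the ability ends.
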(event Ability|EventActivateAbilityFromEvent
  (params (EventData :type Gameplay%20Event%20Data%20Structure%20%28by%20ref%29))
  (bind (_eventtag _instigator _target _optionalobject _optionalobject2 _contexthandle _instigatortags) (Utilities|Struct|BreakGameplayEventData EventData))
  (|CallParentFunction EventData)
  (Variables|Default|SetInstigator (Game|Player|GetPlayerPawn (Fortnite|GetFortPlayerControllerfromActor _instigator)))
  (Variables|Default|SetTarget _target)
  (if (GameplayTags|HasTag _instigatortags "(TagName=\"Gameplay.Damage.Physical.Explosive\")")
    (if (Math|Float|InRange(Float) (.yaw (Transformation|GetActorRotation (Variables|GAVars|Data|GetOwningPlayerPawn))) (.yaw (Math|Rotator|MakeRotator 0.0 0.0 (- (.yaw (Math|Rotator|InvertRotator (Math|Rotator|FindLookatRotation (Transformation|GetActorLocation (Variables|Default|GetInstigator)) (Transformation|GetActorLocation (Variables|Default|GetTarget))))) 6e+01))) (.yaw (Math|Rotator|MakeRotator 0.0 0.0 (+ (.yaw (Math|Rotator|InvertRotator (Math|Rotator|FindLookatRotation (Transformation|GetActorLocation (Variables|Default|GetInstigator)) (Transformation|GetActorLocation (Variables|Default|GetTarget))))) 6e+01))))
      (Quest|SendComplexCustomStatEvent (Variables|GAVars|Data|GetOwningPlayerQuestManager) 0 "(GameplayTags=((TagName=\"Athena.Quests.NoLookExplosive\")))" "(GameplayTags=)" 1)
      (else
        _)))
  (Ability|EndAbility))
